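# Sound cue for a Heist gate closing, playing a single wave through a Modulator that boosts volume at fixed pitch.
# Overridden attenuation uses a box shape with distance-based low-pass filtering, occlusion and stereo spread, alongside a referenced military base door attenuation asset.
# Concurrency allows one instance per owner, stopping the oldest on retrigger.
# The cue is routed to a music stinger sound class.
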SoundCue /HeistBasesGameplay/Audio/Gate/Heist_Gate_Close_Cue
root=Modulator_0
nodes [
 {id=Modulator_0, children [WavePlayer_0], properties {pitchMax=1, pitchMin=1, volumeMax=1.7, volumeMin=1.7}}
 {id=WavePlayer_0, properties {soundWaveAssetPtr=/HeistBasesGameplay/Audio/Gate/wav/Heist_Gate_Close_01}}
]
properties {
 attenuationSettings=/Game/Sounds/Creative/Attenuations/Fort_MilitaryBase_Doors_Att
 bOverrideAttenuation=true
 bOverrideConcurrency=true
 maxDistance=3560
 soundClassObject=/Game/Sounds/FortSoundClassesAndModes/Fort_Music_Stinger_PSM
 volumeMultiplier=0.8
 attenuationOverrides {
  attenuationShape=Box
  bAttenuateWithLPF=true
  bEnableOcclusion=true
  falloffDistance=3000
  lPFFrequencyAtMax=1000
  lPFFrequencyAtMin=15000
  lPFRadiusMax=2000
  lPFRadiusMin=300
  occlusionLowPassFilterFrequency=900
  occlusionTraceChannel=ECC_GameTraceChannel6
  occlusionVolumeAttenuation=0.9
  stereoSpread=230
  attenuationShapeExtents=<310 560 210>
 }
 concurrencyOverrides {bLimitToOwner=true, maxCount=1, resolutionRule=StopOldest, retriggerTime=0.2}
}
